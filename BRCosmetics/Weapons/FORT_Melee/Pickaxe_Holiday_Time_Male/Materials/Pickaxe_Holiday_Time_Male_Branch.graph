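# Masked, two-sided subsurface material for the branch part of the Holiday Time Male pickaxe cosmetic, usable on skeletal meshes and mesh deformers.
# It holds three branch texture samples (D, a normal map and an S mask map) plus a dark green vector parameter.
# The graph records no links between these nodes or to material outputs, and has no math nodes.
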Material /BRCosmetics/Weapons/FORT_Melee/Pickaxe_Holiday_Time_Male/Materials/Pickaxe_Holiday_Time_Male_Branch
graph {
 properties {
  bUsedWithMeshDeformer=true
  bUsedWithSkeletalMesh=true
  blendMode=BLEND_Masked
  pixelDepthOffsetMode=PDOM_Legacy
  shadingModel=MSM_Subsurface
  twoSided=true
  usedShadingModels=Subsurface
 }
 nodes [
  TextureSample() {texture=/BRCosmetics/Weapons/FORT_Melee/Pickaxe_Holiday_Time_Male/Textures/T_Pickaxe_Holiday_Time_Male_Branch_D}
  TextureSample() {
   samplerType=SAMPLERTYPE_Normal
   texture=/BRCosmetics/Weapons/FORT_Melee/Pickaxe_Holiday_Time_Male/Textures/T_Pickaxe_Holiday_Time_Male_Branch_N
  }
  TextureSample() {
   samplerType=SAMPLERTYPE_Masks
   texture=/BRCosmetics/Weapons/FORT_Melee/Pickaxe_Holiday_Time_Male/Textures/T_Pickaxe_Holiday_Time_Male_Branch_S
  }
  VectorParameter() {defaultValue=<rgba 0.0199 0.0417 0.0074 0>}
 ]
}
