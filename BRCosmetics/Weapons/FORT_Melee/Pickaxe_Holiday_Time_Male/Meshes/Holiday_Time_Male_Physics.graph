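# Physics asset for the Holiday Time male pickaxe, a two-handed melee cosmetic with dynamic branches and dangling ornaments.
# A capsule on the handle bone anchors sphere bodies on the branch and ornament bones.
# Main branches are constrained to the handle, two secondary branches chain off branches e and f, and ornaments hang from their branches.
# Swing and tight twist limits keep motion contained, with ornaments allowed looser movement than most branches.
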PhysicsAsset /BRCosmetics/Weapons/FORT_Melee/Pickaxe_Holiday_Time_Male/Meshes/Holiday_Time_Male_Physics
bodies {
 Melee_TwoHanded_Handle [{shapeType=Capsule, length=110.3808, radius=11.2891}]
 dyn_branch_a [{shapeType=Sphere, radius=21.9447, center=<20.9906 1.8037 -1.0699>}]
 dyn_branch_b [{shapeType=Sphere, radius=18.0253, center=<17.3704 7.8039 2.1545>}]
 dyn_branch_c [{shapeType=Sphere, radius=18.7252, center=<17.2912 0.6271 4.2268>}]
 dyn_branch_d2 [{shapeType=Sphere, radius=18.1624, center=<-2.7138 -0.6388 1.1702>}]
 dyn_branch_e [{shapeType=Sphere, radius=36.7323, center=<26.7953 5.3031 7.7618>}]
 dyn_branch_e2 [{shapeType=Sphere, radius=10.7671, center=<16.1496 -10.1445 -3.0522>}]
 dyn_branch_f [{shapeType=Sphere, radius=36.7323, center=<26.7953 5.3031 7.7618>}]
 dyn_branch_f2 [{shapeType=Sphere, radius=18.3144, center=<18.2418 5.3032 -0.9342>}]
 dyn_ornamant_a [{shapeType=Sphere, radius=4.1404, center=<3.948 0.2098 0.0876>}]
 dyn_ornament_b [{shapeType=Sphere, radius=4.8146, center=<3.9314 -0.0244 -0.1706>}]
 dyn_ornament_e [{shapeType=Sphere, radius=3.7318, center=<3.1946 -0.3827 -0.0436>}]
 dyn_ornament_f [{shapeType=Sphere, radius=4.8133, center=<4.553 0.0083 0.0142>}]
}
constraints [
 {bone1Name=dyn_branch_a, bone2Name=Melee_TwoHanded_Handle, swing1LimitDegrees=45, swing2LimitDegrees=45, twistLimitDegrees=5}
 {bone1Name=dyn_branch_b, bone2Name=Melee_TwoHanded_Handle, swing1LimitDegrees=45, swing2LimitDegrees=45, twistLimitDegrees=5}
 {bone1Name=dyn_branch_c, bone2Name=Melee_TwoHanded_Handle, swing1LimitDegrees=45, swing2LimitDegrees=45, twistLimitDegrees=5}
 {bone1Name=dyn_branch_d2, bone2Name=Melee_TwoHanded_Handle, swing1LimitDegrees=45, swing2LimitDegrees=45, twistLimitDegrees=5}
 {bone1Name=dyn_branch_f, bone2Name=Melee_TwoHanded_Handle, swing1LimitDegrees=45, swing2LimitDegrees=45, twistLimitDegrees=5}
 {bone1Name=dyn_branch_e, bone2Name=Melee_TwoHanded_Handle, swing1LimitDegrees=45, swing2LimitDegrees=45, twistLimitDegrees=5}
 {bone1Name=dyn_branch_f2, bone2Name=dyn_branch_f, swing1LimitDegrees=65, swing2LimitDegrees=65, twistLimitDegrees=10}
 {bone1Name=dyn_branch_e2, bone2Name=dyn_branch_e, swing1LimitDegrees=45, swing2LimitDegrees=45, twistLimitDegrees=5}
 {bone1Name=dyn_ornament_f, bone2Name=dyn_branch_f, swing1LimitDegrees=65, swing2LimitDegrees=65, twistLimitDegrees=15}
 {bone1Name=dyn_ornament_e, bone2Name=dyn_branch_e2, swing1LimitDegrees=65, swing2LimitDegrees=65, twistLimitDegrees=15}
 {bone1Name=dyn_ornament_b, bone2Name=dyn_branch_b, swing1LimitDegrees=65, swing2LimitDegrees=65, twistLimitDegrees=15}
 {bone1Name=dyn_ornamant_a, bone2Name=dyn_branch_a, swing1LimitDegrees=65, swing2LimitDegrees=65, twistLimitDegrees=15}
]
